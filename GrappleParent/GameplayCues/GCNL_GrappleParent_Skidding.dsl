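; GCNL_GrappleParent_Skidding is a gameplay cue notify, likely looping, named for the skidding state of GrappleParent.
; It adds no logic of its own: the begin overlap, begin play and end play events are empty, and the construction script only calls its parent.
; The graphs define no effects or cue handlers, so any behavior comes from the parent class or settings not shown here.
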(event Collision|EventActorBeginOverlap
  (params (OtherActor :type Actor%20Object%20Reference)))

(event EventBeginPlay)

(event EventEndPlay
  (params (EndPlayReason :type EEndPlayReason%20Enum :default "Destroyed")))

(fn ConstructionScript
  (|CallParentFunction))
